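# Third-person Sound Cue for the music of the Phonograph emote, using emote attenuation, concurrency and sound class.
# A Mixer layers a boosted one-shot start sound with a looping music wave.
# An Enveloper keeps the loop near silent before fading it in, then endlessly repeats a pitch curve that rises and falls.
# It keeps playing when silent and sends to the music audio-analysis submix.
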SoundCue /Game/Athena/Sounds/Emotes/Phonograph/SC_Emote_Phonograph_3P_Cue
root=Mixer_0
nodes [
 {
  id=Enveloper_1
  children [WavePlayer_1]
  properties {
   bLoop=true
   bLoopIndefinitely=true
   loopEnd=14.63
   loopStart=4.6333
   pitchCurve {editorCurveData {keys [
      <at -0.1194 0.2042>
      {arriveTangent=-0.0009, interpMode=RCIM_Cubic, leaveTangent=-0.0009, tangentMode=RCTM_User, time=4.6333}
      {arriveTangent=0.0517, interpMode=RCIM_Cubic, leaveTangent=0.0517, tangentMode=RCTM_User, time=7.6602, value=1.038}
      {arriveTangent=-0.0007, interpMode=RCIM_Cubic, leaveTangent=-0.0007, tangentMode=RCTM_User, time=9.3608, value=1.186}
      <at 12.384 1.186>
      {arriveTangent=-0.0828, interpMode=RCIM_Cubic, leaveTangent=-0.0828, time=13.5177, value=1.0684}
      {time=14.63}
     ]}}
   volumeCurve {editorCurveData {keys [<at 0 0>,<at 2.6177 0.0046>,<at 4.6333 1>]}}
  }
 }
 {id=Mixer_0, children [WavePlayer_2,Enveloper_1], properties {inputVolume [2,1]}}
 {id=WavePlayer_1, properties {bLooping=true, soundWaveAssetPtr=&/Emote_Phonograph}}
 {id=WavePlayer_2, properties {soundWaveAssetPtr=&/Emote_Phonograph_Start}}
]
properties {
 attenuationSettings=&Fort_Emotes_Athena_Att
 maxDistance=1600
 soundClassObject=/Game/Sounds/FortSoundClassesAndModes/Fort_Music_Emote_3P
 virtualizationMode=PlayWhenSilent
 concurrencySet [/Game/Sounds/Concurrency_Rules/Athena_EmoteMusic_Conc]
 soundSubmixSends [{minSendLevel=0, soundSubmix=&Fortnite_AudioAnalysis_Music}]
}
